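# Sound cue for the house collapse in Section 01 of the PistolCanary event, judging by its name and the HouseOfCardsFall wave it plays.
# A single wave player feeds the root directly, with no mixing, modulation or randomization.
# It routes through the Fort_Events sound class with a boosted volume and uses the PistolCanary POI house attenuation asset for spatial falloff.
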SoundCue /PistolCanary_Event_Audio/Assets/Section01/House/SC_PistolCanary_Section01_HouseCollapse
root=WavePlayer_1
nodes [{id=WavePlayer_1, properties {soundWaveAssetPtr=/PistolCanary_Event_Audio/Assets/Section01/wav/PistolCanary_POI_HouseOfCardsFall}}]
properties {
 attenuationSettings=/PistolCanary_Audio/Assets/Settings/ATT_PistolCanary_POI_House
 maxDistance=150000
 soundClassObject=&Fort_Events
 volumeMultiplier=1.5
}
